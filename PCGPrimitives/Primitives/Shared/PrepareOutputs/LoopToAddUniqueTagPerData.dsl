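; PCG loop-body subgraph that gives each looped data item its own tag from a parallel TagForEachData input.
; The current loop index selects the matching entry from TagForEachData, which is then filtered to param (attribute set) data.
; That result drives the TagsToAdd override of Add Tags, which tags the incoming data and passes it to the output.
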(graph LoopToAddUniqueTagPerData
  (inputs (TagForEachData :type Any))
  (bind GetLoopIndex_0 (Get%20Loop%20Index))
  (bind FilterDataByIndex_0 (Filter%20Data%20By%20Index))
  (bind FilterDataByType_0 (Filter%20Data%20By%20Type :TargetType (struct :Ids [(struct :Struct /Script/PCG.PCGDataTypeInfoParam)] :CustomSubtype 9)))
  (bind AddTags_0 (Add%20Tags))
  (connect DefaultInputNode In AddTags_0 In)
  (connect FilterDataByType_0 InsideFilter AddTags_0 TagsToAdd)
  (connect DefaultInputNode TagForEachData FilterDataByIndex_0 In)
  (connect GetLoopIndex_0 Out FilterDataByIndex_0 SelectedIndices)
  (connect FilterDataByIndex_0 InsideFilter FilterDataByType_0 In)
  (output Out AddTags_0))
