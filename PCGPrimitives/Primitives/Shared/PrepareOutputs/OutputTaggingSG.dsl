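; PCG subgraph that retags primitive output data with its type and primitive name while preserving where it came from.
; It strips incoming tags except Overriden, then conditionally adds tags derived from the DataType and Primitive parameters.
; In parallel, it backs up each input's _ID tags as a joined attribute, using an index tag to restore data order after filtering.
; That backup is copied onto the retagged data and converted back into tags to track parentage.
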(graph OutputTaggingSG
  (params (DataType :type String :default "None") (Primitive :type String :default "None"))
  (bind CreateConstant_1 (Create%20Constant :AttributeTypes (struct :Type String :StringValue "Overriden,")))
  (bind DeleteTags_1 (Delete%20Tags :Operation KeepOnlySelectedTags))
  ;[0 0 0 0.5] Type
  (bind AttributeStringOp_6 (Attribute%20String%20Op :InB DataType))
  (bind AttributeCompareOp_0 (Attribute%20Compare%20Op :InA DataType))
  (bind AttributeCompareOp_1 (Attribute%20Compare%20Op :InA DataType))
  (bind AttributeBooleanOp_0 (Attribute%20Boolean%20Op :Operation Or))
  (bind Branch_0 (Branch))
  (bind AddTags_0 (Add%20Tags :TagsToAdd "Cluster"))
  ;[0 0 0 0.5] Input Class_ID tags backup to persist onto produced output to track data creation origins through its parenting
  (bind GetTags_1 (Get%20Tags :bExtractTagValues true))
  (bind AttributeFilter_0 (Filter%20Attribute%20Elements :AttributeTypes (struct :Type String :StringValue "_ID") :bUseConstantThreshold true :Operator Substring :TargetAttribute Attribute :ThresholdAttribute @Last))
  (bind SortAttributes_1 (Sort%20Attributes :bUseStableSort false :SortAttributes [(struct :InputSource $Index :SortMethod Descending)]))
  (bind AttributeStringOp_9 (Attribute%20String%20Op :InputSource1 Attribute))
  (bind AttributeStringOp_10 (Attribute%20String%20Op :InputSource1 Attribute :InputSource2 Values))
  (bind WriteDataIndex_2 (Write%20Data%20Index :IndexTag "TagsBackedUpIndex"))
  (bind FilterDataByAttribute_0 (Filter%20Data%20By%20Attribute :FilterMode FilterByValue :FilterOperator NotEqual :TargetAttribute Attribute :Threshold (struct :AttributeTypes (struct :Type String))))
  (bind AttributeReduce_1 (Attribute%20Reduce :bWriteToDataDomain true :InputSource Attribute :Operation Join :OutputAttributeName TagsBackedUp))
  (bind AddAttribute_0 (Add%20Attribute :AttributeTypes (struct :Type String) :OutputTarget @Data.TagsBackedUp))
  (bind Gather_0 (Gather))
  (bind SortTags_0 (Sort%20Data%20By%20Tag%20Value :Tag TagsBackedUpIndex))
  ;[0 0 0 0.5] Primitive
  (bind AttributeStringOp_13 (Attribute%20String%20Op :InB Primitive))
  (bind AttributeCompareOp_5 (Attribute%20Compare%20Op :InA Primitive))
  (bind AttributeCompareOp_6 (Attribute%20Compare%20Op :InA Primitive))
  (bind AttributeBooleanOp_3 (Attribute%20Boolean%20Op :Operation Or))
  (bind Branch_5 (Branch))
  (bind AddTags_3 (Add%20Tags :TagsToAdd "Cluster"))
  ;[0 0 0 0.5] Write Previous Class_ID to identify the data origin
  (bind CopyAttributes_0 (Copy%20Attributes :InputSource @Data.TagsBackedUp :OutputTarget @Data.TagsBackedUp))
  (bind DataAttributesToTags_1 (Data%20Attributes%20To%20Tags :AttributesTagsMapping (map ("TagsBackedUp" @Source)) :bDeleteInputsAfterOperation true))
  (connect FilterDataByAttribute_0 OutsideFilter AddAttribute_0 In)
  (connect Branch_0 Output%20A AddTags_0 In)
  (connect AttributeStringOp_6 Out AddTags_0 TagsToAdd)
  (connect Branch_5 Output%20A AddTags_3 In)
  (connect AttributeStringOp_13 Out AddTags_3 TagsToAdd)
  (connect AttributeCompareOp_0 Out AttributeBooleanOp_0 InA)
  (connect AttributeCompareOp_1 Out AttributeBooleanOp_0 InB)
  (connect AttributeCompareOp_5 Out AttributeBooleanOp_3 InA)
  (connect AttributeCompareOp_6 Out AttributeBooleanOp_3 InB)
  (connect GetTags_1 Out AttributeFilter_0 In)
  (connect FilterDataByAttribute_0 InsideFilter AttributeReduce_1 In)
  (connect SortAttributes_1 Out AttributeStringOp_9 InA)
  (connect AttributeStringOp_9 Out AttributeStringOp_10 InA)
  (connect AttributeStringOp_9 Out AttributeStringOp_10 InB)
  (connect AttributeBooleanOp_0 Out Branch_0 bOutputToB)
  (connect DeleteTags_1 Out Branch_0 In)
  (connect AttributeBooleanOp_3 Out Branch_5 bOutputToB)
  (connect AddTags_0 Out Branch_5 In)
  (connect Branch_0 Output%20B Branch_5 In)
  (connect SortTags_0 Out CopyAttributes_0 Source)
  (connect AddTags_3 Out CopyAttributes_0 Target)
  (connect Branch_5 Output%20B CopyAttributes_0 Target)
  (connect CopyAttributes_0 Out DataAttributesToTags_1 In)
  (connect DefaultInputNode In DeleteTags_1 In)
  (connect CreateConstant_1 Out DeleteTags_1 SelectedTags)
  (connect WriteDataIndex_2 Out FilterDataByAttribute_0 In)
  (connect AttributeReduce_1 Out Gather_0 In)
  (connect AddAttribute_0 Out Gather_0 In2)
  (connect DefaultInputNode In GetTags_1 In)
  (connect AttributeFilter_0 InsideFilter SortAttributes_1 In)
  (connect Gather_0 Out SortTags_0 In)
  (connect AttributeStringOp_10 Out WriteDataIndex_2 In)
  (output Out DataAttributesToTags_1))
