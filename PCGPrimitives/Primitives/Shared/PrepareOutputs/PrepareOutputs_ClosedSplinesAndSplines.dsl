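; PCG subgraph that prepares a primitive's spline output, keeping only splines tagged as closed or regular.
; Input drops empty and non-spline data, forces counterclockwise direction, and strips attributes with a keep-selected delete.
; Chained tag filters separate Type:ClosedSpline, then Type:Spline from the remainder, discarding anything else.
; Each set passes a count-driven branch, gains tags from the Primitive parameter, and is type-filtered into the shared output.
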(graph PrepareOutputs_ClosedSplinesAndSplines
  (params (Primitive :type String :default "None"))
  (bind RemoveEmptyData_1 (Remove%20Empty%20Data))
  ;[0 0 0 0.5] Tag and apply Metadata for Output
  (bind FilterDataByType_2 (Filter%20Data%20By%20Type :TargetType (struct :Ids [(struct :Struct /Script/PCG.PCGDataTypeInfoSpline)])))
  (bind SplineDirection_1 (Spline%20Direction :Operation ForceCounterClockwise))
  (bind DeleteAttributes_1 (Delete%20Attributes :MetadataDomain None :Operation KeepSelectedAttributes))
  (bind FilterDataByTag_0 (Filter%20Data%20By%20Tag :SelectedTags "Type:ClosedSpline"))
  (bind DataNum_2 (Data%20Count))
  (bind Branch_2 (Branch))
  (bind AddTags_0 (Add%20Tags :TagsToAdd-param Primitive))
  (bind FilterDataByType_1 (Filter%20Data%20By%20Type :TargetType (struct :Ids [(struct :Struct /Script/PCG.PCGDataTypeInfoSpline)])))
  (bind FilterDataByTag_1 (Filter%20Data%20By%20Tag :SelectedTags "Type:Spline"))
  (bind DataNum_0 (Data%20Count))
  (bind Branch_0 (Branch))
  (bind AddTags_1 (Add%20Tags :TagsToAdd-param Primitive))
  (bind FilterDataByType_0 (Filter%20Data%20By%20Type :TargetType (struct :Ids [(struct :Struct /Script/PCG.PCGDataTypeInfoSpline)])))
  (connect Branch_2 Output%20B AddTags_0 In)
  (connect Branch_0 Output%20B AddTags_1 In)
  (connect DataNum_0 Out Branch_0 bOutputToB)
  (connect FilterDataByTag_1 InsideFilter Branch_0 In)
  (connect DataNum_2 Out Branch_2 bOutputToB)
  (connect FilterDataByTag_0 InsideFilter Branch_2 In)
  (connect FilterDataByTag_1 InsideFilter DataNum_0 In)
  (connect FilterDataByTag_0 InsideFilter DataNum_2 In)
  (connect SplineDirection_1 Out DeleteAttributes_1 In)
  (connect DeleteAttributes_1 Out FilterDataByTag_0 In)
  (connect FilterDataByTag_0 OutsideFilter FilterDataByTag_1 In)
  (connect AddTags_1 Out FilterDataByType_0 In)
  (connect AddTags_0 Out FilterDataByType_1 In)
  (connect RemoveEmptyData_1 Out FilterDataByType_2 In)
  (connect DefaultInputNode In RemoveEmptyData_1 In)
  (connect FilterDataByType_2 InsideFilter SplineDirection_1 In)
  (output Out FilterDataByType_0.InsideFilter)
  (output Out FilterDataByType_1.InsideFilter))
