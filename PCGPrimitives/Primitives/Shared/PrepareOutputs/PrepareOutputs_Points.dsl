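; PCG graph PrepareOutputs_Points, a shared step that cleans data and tags it for output under a given Primitive name.
; It removes empty data, then runs a Delete Attributes pass in keep-selected mode with no attributes listed.
; A data count drives a Branch, so only a non-empty result reaches the OutputTaggingSG subgraph, which receives the Primitive string and feeds the graph output.
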(graph PrepareOutputs_Points
  (params (Primitive :type String))
  (bind RemoveEmptyData_0 (Remove%20Empty%20Data))
  (bind DeleteAttributes_2 (Delete%20Attributes :MetadataDomain None :Operation KeepSelectedAttributes))
  (bind DataNum_0 (Data%20Count))
  (bind Branch_0 (Branch))
  (bind Subgraph_1 (Subgraph :Graph /PCGPrimitives/Primitives/Shared/PrepareOutputs/OutputTaggingSG.OutputTaggingSG :Primitive Primitive))
  (connect DataNum_0 Out Branch_0 bOutputToB)
  (connect DeleteAttributes_2 Out Branch_0 In)
  (connect DeleteAttributes_2 Out DataNum_0 In)
  (connect RemoveEmptyData_0 Out DeleteAttributes_2 In)
  (connect DefaultInputNode In RemoveEmptyData_0 In)
  (connect Branch_0 Output%20B Subgraph_1 In)
  (output Out Subgraph_1))
